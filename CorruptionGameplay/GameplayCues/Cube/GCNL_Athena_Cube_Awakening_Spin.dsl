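; Looping gameplay cue for the Cube awakening that draws a beam between two corruption cubes, the target ("Mom") and the effect causer ("Dad").
; FindBestSocket picks each cube's socket A-F best facing the other by dot product, anchoring the beam particle's start and end.
; The SphereMove timeline drives Zap and SphereOffset material parameters toward those sockets alongside endpoint audio.
; Removal reverses the timeline, fades the loops and plays stop sounds.
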(event EventOnLoopingStart(EitherVFX-preferthisversion)
  (params (MyTarget :type Actor%20Object%20Reference) (Parameters :type Gameplay%20Cue%20Parameters%20Structure%20%28by%20ref%29) (ParticleComponents :type Array%20of%20FXSystem%20Component%20Object%20References) (AudioComponents :type Array%20of%20Audio%20Component%20Object%20References))
  (bind (_normalizedmagnitude _rawmagnitude _effectcontext _matchedtagname _originaltag _aggregatedsourcetags _aggregatedtargettags _location _normal _instigator _effectcauser) (Ability|GameplayCue|BreakGameplayCueParameters Parameters))
  (bind _output (Utilities|Array|Get(acopy) ParticleComponents))
  (bind _momcubemesh (Variables|Default|GetMomCubeMesh))
  (bind _dadcubemesh (Variables|Default|GetDadCubeMesh))
  (bind _asbga_athena_corruption_gameplay_cube (Utilities|Casting|CastToBGA_Athena_CorruptionGameplay_Cube MyTarget))
  (Variables|Default|SetMomCubeMesh (Variables|/CorruptionGameplay/Gameplay/Environmental/Cube/BGA_Athena_CorruptionGameplay_Cube.BGA_Athena_CorruptionGameplay_Cube_C|GetStaticMesh _asbga_athena_corruption_gameplay_cube))
  (bind _asbga_athena_corruption_gameplay_cube_2 (Utilities|Casting|CastToBGA_Athena_CorruptionGameplay_Cube _effectcauser))
  (Variables|Default|SetDadCubeMesh (Variables|/CorruptionGameplay/Gameplay/Environmental/Cube/BGA_Athena_CorruptionGameplay_Cube.BGA_Athena_CorruptionGameplay_Cube_C|GetStaticMesh _asbga_athena_corruption_gameplay_cube_2))
  (Utilities|FlowControl|Sequence
    (:then_0
      (bind _nearestsocket (CallFunction|FindBestSocket _momcubemesh (Math|Vector|GetForwardVector (Math|Rotator|FindLookatRotation (Transformation|GetActorLocation _asbga_athena_corruption_gameplay_cube) (Transformation|GetActorLocation _asbga_athena_corruption_gameplay_cube_2)))))
      (bind _output_get (Variables|Default|SetNearestSocketMom _nearestsocket))
      (bind (_returnvalue_location _returnvalue_rotation) (Math|Transform|BreakTransform (Transformation|GetSocketTransform _momcubemesh _output_get "RTS_World")))
      (Variables|Default|SetMomsSocketPosition _returnvalue_location)
      (Transformation|SetWorldLocationAndRotation _output _returnvalue_location _returnvalue_rotation false false))
    (:then_1
      (bind _nearestsocket_2 (CallFunction|FindBestSocket _dadcubemesh (* (Math|Vector|GetForwardVector (Math|Rotator|FindLookatRotation (Transformation|GetActorLocation _asbga_athena_corruption_gameplay_cube) (Transformation|GetActorLocation _asbga_athena_corruption_gameplay_cube_2))) -1.0)))
      (bind _output_get_2 (Variables|Default|SetNearestSocketDad _nearestsocket_2))
      (bind (_returnvalue_location_2 _returnvalue_rotation_2) (Math|Transform|BreakTransform (Transformation|GetSocketTransform _dadcubemesh _output_get_2 "RTS_World")))
      (bind _output_get_3 (Variables|Default|SetDadsSocketPosition _returnvalue_location_2))
      (Effects|Components|ParticleSystem|SetVectorParameter _output "EndPoint" _output_get_3)
      (Effects|Components|ParticleSystem|SetVectorParameter _output "EndNormal" (Math|Rotator|GetRotationXVector _returnvalue_rotation_2))
      (Effects|Components|ParticleSystem|SetVectorParameter _output "EndUp" (Math|Vector|GetUpVector _returnvalue_rotation_2)))
    (:then_2
      (CallFunction|SpawnLoopAudio))
    (:then_3
      (bind (_direction _newtrack_0) (Timeline|SphereMove PlayFromStart)
        (:Update
          (bind _momcubemesh_2 (Variables|Default|GetMomCubeMesh)
            (:then
              (Rendering|Material|SetScalarParameterValueonMaterials _momcubemesh_2 "Zap" _newtrack_0)
              (Rendering|Material|SetVectorParameterValueonMaterials _momcubemesh_2 "SphereOffset" (* (- (Transformation|GetSocketLocation _momcubemesh_2 (Variables|Default|GetNearestSocketMom)) (Transformation|GetWorldLocation _momcubemesh_2)) _newtrack_0)))
            (:else
              _))
          (bind _dadcubemesh_2 (Variables|Default|GetDadCubeMesh)
            (:then
              (Rendering|Material|SetScalarParameterValueonMaterials _dadcubemesh_2 "Zap" _newtrack_0)
              (Rendering|Material|SetVectorParameterValueonMaterials _dadcubemesh_2 "SphereOffset" (* (- (Transformation|GetSocketLocation _dadcubemesh_2 (Variables|Default|GetNearestSocketDad)) (Transformation|GetWorldLocation _dadcubemesh_2)) _newtrack_0)))
            (:else)))))))

(event EventOnRemoval(EitherVFX-preferthisversion)
  (params (MyTarget :type Actor%20Object%20Reference) (Parameters :type Gameplay%20Cue%20Parameters%20Structure%20%28by%20ref%29) (ParticleComponents :type Array%20of%20FXSystem%20Component%20Object%20References) (AudioComponents :type Array%20of%20Audio%20Component%20Object%20References) (BurstCameraShakeInstance :type Camera%20Shake%20Base%20Object%20Reference) (BurstDecalInstance :type Decal%20Actor%20Object%20Reference))
  (CallFunction|SpawnLoopStopAudio)
  (Timeline|SphereMove Reverse))

(fn FindBestSocket
  (params (Mesh :type Scene%20Component%20Object%20Reference) (LookAtDirection :type Vector))
  (LocalVariables|SetNearest -1.0)
  (for _array_element (Utilities|Array|MakeArray "A" "B" "C" "D" "E" "F")
    (if (> (Math|Vector|DotProduct (Math|Vector|GetForwardVector (Transformation|GetSocketRotation Mesh _array_element)) LookAtDirection) (LocalVariables|GetNearest))
      (LocalVariables|SetNearest (Math|Vector|DotProduct (Math|Vector|GetForwardVector (Transformation|GetSocketRotation Mesh _array_element)) LookAtDirection))
      (LocalVariables|SetSocket _array_element)))
  (return (LocalVariables|GetSocket)))

(fn SpawnLoopAudio
  (bind _momssocketposition (Variables|Default|GetMomsSocketPosition))
  (bind _dadssocketposition (Variables|Default|GetDadsSocketPosition))
  (Utilities|FlowControl|Sequence
    (:then_0
      (bind _returnvalue (Audio|SpawnSoundatLocation 0 (Variables|Default|GetLoop_StartPoint) _momssocketposition))
      (Variables|Default|SetStartPointLoop _returnvalue)
      (Audio|PlaySoundatLocation 0 "/CorruptionGameplay/Gameplay/Environmental/Cube/Audio/Cube_Awaken_Beam_StartPoint_Cue.Cube_Awaken_Beam_StartPoint_Cue" _momssocketposition))
    (:then_1
      (bind _returnvalue_2 (Audio|SpawnSoundatLocation 0 (Variables|Default|GetLoop_EndPoint) _dadssocketposition))
      (Variables|Default|SetStopPointLoop _returnvalue_2)
      (Audio|PlaySoundatLocation 0 "/CorruptionGameplay/Gameplay/Environmental/Cube/Audio/Cube_Awaken_Beam_EndPoint_Cue.Cube_Awaken_Beam_EndPoint_Cue" _dadssocketposition))))

(fn SpawnLoopStopAudio
  (Utilities|FlowControl|Sequence
    (:then_0
      (Audio|SpawnSoundAttached (Variables|Default|GetLoopStop_StartPoint) (Variables|Default|GetDadCubeMesh))
      (Audio|SpawnSoundAttached (Variables|Default|GetLoopStop_EndPoint) (Variables|Default|GetMomCubeMesh)))
    (:then_1
      (bind _startpointloop (Variables|Default|GetStartPointLoop)
        (:then
          (Audio|Components|Audio|FadeOut _startpointloop 0.5 0.0 "Linear" :!class /Script/Engine.AudioComponent))
        (:else)))
    (:then_2
      (bind _stoppointloop (Variables|Default|GetStopPointLoop)
        (:then
          (Audio|Components|Audio|FadeOut _stoppointloop 0.5 0.0 "Linear" :!class /Script/Engine.AudioComponent))
        (:else)))))

(fn ConstructionScript)
