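; Looping gameplay cue for the Corruption cube's hover, placing its effects on the cube's downward-facing side.
; On loop start, it acts immediately if the cube's begin-play skip has occurred, otherwise waits for that delegate.
; FindBestSocket dot-products sockets A–F against world down to pick one, and leaves a debug print.
; The cue's particle component snaps to that socket, and a roll-dust Niagara system spawns attached there.
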(event EventOnLoopingStart(EitherVFX-preferthisversion)
  (params (MyTarget :type Actor%20Object%20Reference) (Parameters :type Gameplay%20Cue%20Parameters%20Structure%20%28by%20ref%29) (ParticleComponents :type Array%20of%20FXSystem%20Component%20Object%20References) (AudioComponents :type Array%20of%20Audio%20Component%20Object%20References))
  (bind _asbga_athena_corruption_gameplay_cube (Utilities|Casting|CastToBGA_Athena_CorruptionGameplay_Cube MyTarget))
  (if (Variables|/CorruptionGameplay/Gameplay/Environmental/Cube/BGA_AthenaCorruptionGameplay_StaticPathFollowingCube.BGA_AthenaCorruptionGameplay_StaticPathFollowingCube_C|GetBeginPlaySkipOccured _asbga_athena_corruption_gameplay_cube)
    (bind _nearestsocket (CallFunction|FindBestSocket (Variables|/CorruptionGameplay/Gameplay/Environmental/Cube/BGA_Athena_CorruptionGameplay_Cube.BGA_Athena_CorruptionGameplay_Cube_C|GetStaticMesh _asbga_athena_corruption_gameplay_cube) "0, 0,-1.000000"))
    (Transformation|AttachComponentToComponent (Utilities|Array|Get(acopy) ParticleComponents) (Variables|/CorruptionGameplay/Gameplay/Environmental/Cube/BGA_Athena_CorruptionGameplay_Cube.BGA_Athena_CorruptionGameplay_Cube_C|GetStaticMesh _asbga_athena_corruption_gameplay_cube) _nearestsocket "SnapToTarget" "SnapToTarget" "KeepWorld" true)
    (Effects|SpawnFXSystemAttached :FXSystemAsset "/CorruptionGameplay/Gameplay/Environmental/Cube/FX/NS_Corruption_Cube_Roll_Dust.NS_Corruption_Cube_Roll_Dust" :AttachToComponent (Variables|/CorruptionGameplay/Gameplay/Environmental/Cube/BGA_Athena_CorruptionGameplay_Cube.BGA_Athena_CorruptionGameplay_Cube_C|GetStaticMesh _asbga_athena_corruption_gameplay_cube) :AttachPointName _nearestsocket :Scale "1.000000,1.000000,1.000000" :LocationType "SnapToTarget" :bAutoDestroy true :PoolingMethod "None" :bAutoActivate true)
    (else
      (Default|AssignOnBeginPlaySkipOccured _asbga_athena_corruption_gameplay_cube (delegate Custom|OnBeginPlaySkipOccured)))))

(event Custom|OnBeginPlaySkipOccured
  (bind _nearestsocket (CallFunction|FindBestSocket (Variables|/CorruptionGameplay/Gameplay/Environmental/Cube/BGA_Athena_CorruptionGameplay_Cube.BGA_Athena_CorruptionGameplay_Cube_C|GetStaticMesh (stale EventOnLoopingStart(EitherVFX-preferthisversion) _asbga_athena_corruption_gameplay_cube)) "0, 0,-1.000000"))
  (Transformation|AttachComponentToComponent (Utilities|Array|Get(acopy) (stale EventOnLoopingStart(EitherVFX-preferthisversion) ParticleComponents)) (Variables|/CorruptionGameplay/Gameplay/Environmental/Cube/BGA_Athena_CorruptionGameplay_Cube.BGA_Athena_CorruptionGameplay_Cube_C|GetStaticMesh (stale EventOnLoopingStart(EitherVFX-preferthisversion) _asbga_athena_corruption_gameplay_cube)) _nearestsocket "SnapToTarget" "SnapToTarget" "KeepWorld" true)
  (Effects|SpawnFXSystemAttached :FXSystemAsset "/CorruptionGameplay/Gameplay/Environmental/Cube/FX/NS_Corruption_Cube_Roll_Dust.NS_Corruption_Cube_Roll_Dust" :AttachToComponent (Variables|/CorruptionGameplay/Gameplay/Environmental/Cube/BGA_Athena_CorruptionGameplay_Cube.BGA_Athena_CorruptionGameplay_Cube_C|GetStaticMesh (stale EventOnLoopingStart(EitherVFX-preferthisversion) _asbga_athena_corruption_gameplay_cube)) :AttachPointName _nearestsocket :Scale "1.000000,1.000000,1.000000" :LocationType "SnapToTarget" :bAutoDestroy true :PoolingMethod "None" :bAutoActivate true))

(fn FindBestSocket
  (params (Mesh :type Scene%20Component%20Object%20Reference) (LookAtDirection :type Vector))
  (LocalVariables|SetNearest -1.0)
  (for _array_element (Utilities|Array|MakeArray "A" "B" "C" "D" "E" "F")
    (if (> (Math|Vector|DotProduct (Math|Vector|GetForwardVector (Transformation|GetSocketRotation Mesh _array_element)) LookAtDirection) (LocalVariables|GetNearest))
      (LocalVariables|SetNearest (Math|Vector|DotProduct (Math|Vector|GetForwardVector (Transformation|GetSocketRotation Mesh _array_element)) LookAtDirection))
      (LocalVariables|SetSocket _array_element)))
  (Development|PrintString 0 (Utilities|String|Append "GCN cue hover find socket: " (Utilities|String|ToString(Name) (stale _array_element)) " cube the transform:" (Utilities|String|ToString(Transform) (Transformation|GetActorTransform (Components|GetOwner Mesh))) "" "") true true "(R=1.000000,G=0.015401,B=0.000000,A=1.000000)" 6e+01)
  (return (LocalVariables|GetSocket)))

(fn ConstructionScript)
