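; Looping gameplay cue that plays rolling VFX and audio for the Corruption gameplay cube.
; On start, a valid particle component attaches to the cube mesh's ground socket with a scale-derived radius, and the roll sound spawns attached to the mesh.
; Sound parameters take Size from a scale lookup map and Color from the cue's ability level.
; A looping sound indicator also starts, and removal stops it and fades out the audio.
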(event EventOnLoopingStart(EitherVFX-preferthisversion)
  (params (MyTarget :type Actor%20Object%20Reference) (Parameters :type Gameplay%20Cue%20Parameters%20Structure%20%28by%20ref%29) (ParticleComponents :type Array%20of%20FXSystem%20Component%20Object%20References) (AudioComponents :type Array%20of%20Audio%20Component%20Object%20References))
  (bind _returnvalue_x (Math|Vector|BreakVector (Transformation|GetActorScale3D MyTarget)))
  (bind (_normalizedmagnitude _rawmagnitude _effectcontext _matchedtagname _originaltag _aggregatedsourcetags _aggregatedtargettags _location _normal _instigator _effectcauser _sourceobject _physicalmaterial _gameplayeffectlevel _abilitylevel) (Ability|GameplayCue|BreakGameplayCueParameters Parameters))
  (bind _output (Utilities|Array|Get(acopy) ParticleComponents))
  (bind _asbga_athena_corruption_gameplay_cube (Utilities|Casting|CastToBGA_Athena_CorruptionGameplay_Cube MyTarget))
  (Utilities|FlowControl|Sequence
    (:then_0
      (Utilities|IsValid _output
        (:Is%20Valid
          (Transformation|AttachComponentToComponent _output (Variables|/CorruptionGameplay/Gameplay/Environmental/Cube/BGA_Athena_CorruptionGameplay_Cube.BGA_Athena_CorruptionGameplay_Cube_C|GetStaticMesh _asbga_athena_corruption_gameplay_cube) (Variables|/CorruptionGameplay/Gameplay/Environmental/Cube/BGA_Athena_CorruptionGameplay_Cube.BGA_Athena_CorruptionGameplay_Cube_C|GetGroundSocket _asbga_athena_corruption_gameplay_cube) "SnapToTarget" "SnapToTarget" "KeepWorld" true)
          (Effects|Components|ParticleSystem|SetFloatParameter _output "CubeRadius" (* _returnvalue_x 768.0))
          (bind _returnvalue (Audio|SpawnSoundAttached (Variables|Default|GetRollSound) (Variables|/CorruptionGameplay/Gameplay/Environmental/Cube/BGA_Athena_CorruptionGameplay_Cube.BGA_Athena_CorruptionGameplay_Cube_C|GetStaticMesh _asbga_athena_corruption_gameplay_cube)))
          (bind _output_get (Variables|Default|SetRollAudio _returnvalue))
          (Audio|Parameter|SetIntegerParameter _output_get "Size" (Utilities|Map|Find (Variables|Default|GetScaleMap) (Math|Float|Truncate (* _returnvalue_x 1e+01))) :!class /Script/AudioExtensions.AudioParameterControllerInterface)
          (Audio|Parameter|SetIntegerParameter _output_get "Color" _abilitylevel :!class /Script/AudioExtensions.AudioParameterControllerInterface))))
    (:then_1
      (SoundIndicator|StartSoundIndicatorLooping (Variables|/CorruptionGameplay/Gameplay/Environmental/Cube/BGA_Athena_CorruptionGameplay_Cube.BGA_Athena_CorruptionGameplay_Cube_C|GetFortSoundIndicator _asbga_athena_corruption_gameplay_cube) (Transformation|GetActorLocation _asbga_athena_corruption_gameplay_cube) (Variables|Default|GetRollSound) 1.0 0))))

(event EventOnRemoval(EitherVFX-preferthisversion)
  (params (MyTarget :type Actor%20Object%20Reference) (Parameters :type Gameplay%20Cue%20Parameters%20Structure%20%28by%20ref%29) (ParticleComponents :type Array%20of%20FXSystem%20Component%20Object%20References) (AudioComponents :type Array%20of%20Audio%20Component%20Object%20References) (BurstCameraShakeInstance :type Camera%20Shake%20Base%20Object%20Reference) (BurstDecalInstance :type Decal%20Actor%20Object%20Reference))
  (Utilities|FlowControl|Sequence
    (:then_0
      (bind _rollaudio (Variables|Default|GetRollAudio)
        (:then
          (Audio|Components|Audio|FadeOut _rollaudio 0.5 0.0 "Linear" :!class /Script/Engine.AudioComponent))
        (:else)))
    (:then_1
      (bind _asbga_athena_corruption_gameplay_cube (Utilities|Casting|CastToBGA_Athena_CorruptionGameplay_Cube MyTarget))
      (SoundIndicator|StopSoundIndicatorLooping (Variables|/CorruptionGameplay/Gameplay/Environmental/Cube/BGA_Athena_CorruptionGameplay_Cube.BGA_Athena_CorruptionGameplay_Cube_C|GetFortSoundIndicator _asbga_athena_corruption_gameplay_cube)))))

(fn ConstructionScript)
