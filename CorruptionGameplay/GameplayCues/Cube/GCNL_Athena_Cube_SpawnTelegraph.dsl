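; Looping gameplay cue that telegraphs a Corruption cube spawn with Niagara VFX and sound linking a "dad" cube (effect causer) to a "baby" cube (target).
; FindBestSocket picks face sockets by dot product, anchoring the effect on the baby's downward face and feeding the dad's baby-facing socket in as source.
; A random seed offset varies the lightning.
; Audio loops at both sockets; removal fades them, plays stop sounds and fades any decal.
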(event EventOnLoopingStart(EitherVFX-preferthisversion)
  (params (MyTarget :type Actor%20Object%20Reference) (Parameters :type Gameplay%20Cue%20Parameters%20Structure%20%28by%20ref%29) (ParticleComponents :type Array%20of%20FXSystem%20Component%20Object%20References) (AudioComponents :type Array%20of%20Audio%20Component%20Object%20References))
  (bind _returnvalue_x (Math|Vector|BreakVector (Transformation|GetActorScale3D MyTarget)))
  (bind (_normalizedmagnitude _rawmagnitude _effectcontext _matchedtagname _originaltag _aggregatedsourcetags _aggregatedtargettags _location _normal _instigator _effectcauser) (Ability|GameplayCue|BreakGameplayCueParameters Parameters))
  (bind _output (Utilities|Array|Get(acopy) ParticleComponents))
  (bind _babycubemesh (Variables|Default|GetBabyCubeMesh))
  (bind _dadcubemesh (Variables|Default|GetDadCubeMesh))
  (bind _dadcubemesh_2 (Variables|Default|GetDadCubeMesh))
  (Utilities|FlowControl|Sequence
    (:then_0
      (bind _spawneddecal (Variables|Default|GetSpawnedDecal)
        (:then
          (Components|DestroyComponent _spawneddecal 0))
        (:else)))
    (:then_1
      (Effects|Components|ParticleSystem|SetFloatParameter _output "CubeRadius" (* _returnvalue_x 768.0))
      (bind _asniagara_particle_system_component (Utilities|Casting|CastToNiagaraComponent _output))
      (Niagara|SetRandomSeedOffset _asniagara_particle_system_component (Math|Random|RandomInteger 0)) ; Needed for randomization of lightning spawns
    )
    (:then_2
      (bind _asbga_athena_corruption_gameplay_cube (Utilities|Casting|CastToBGA_Athena_CorruptionGameplay_Cube MyTarget))
      (Variables|Default|SetBabyCubeMesh (Variables|/CorruptionGameplay/Gameplay/Environmental/Cube/BGA_Athena_CorruptionGameplay_Cube.BGA_Athena_CorruptionGameplay_Cube_C|GetStaticMesh _asbga_athena_corruption_gameplay_cube))
      (bind _asbga_athena_corruption_gameplay_cube_2 (Utilities|Casting|CastToBGA_Athena_CorruptionGameplay_Cube _effectcauser))
      (Variables|Default|SetDadCubeMesh (Variables|/CorruptionGameplay/Gameplay/Environmental/Cube/BGA_Athena_CorruptionGameplay_Cube.BGA_Athena_CorruptionGameplay_Cube_C|GetStaticMesh _asbga_athena_corruption_gameplay_cube_2))
      (Utilities|FlowControl|Sequence
        (:then_0
          (bind _nearestsocket (CallFunction|FindBestSocket _babycubemesh "0, 0,-1.000000"))
          (bind (_returnvalue_location _returnvalue_rotation) (Math|Transform|BreakTransform (Transformation|GetSocketTransform _babycubemesh _nearestsocket "RTS_World")))
          (bind _output_get (Variables|Default|SetBabySocketPosition _returnvalue_location))
          (Transformation|SetWorldLocationAndRotation _output _output_get _returnvalue_rotation false false))
        (:then_1
          (bind _nearestsocket_2 (CallFunction|FindBestSocket _dadcubemesh (* (Math|Vector|GetForwardVector (Math|Rotator|FindLookatRotation (Transformation|GetActorLocation _asbga_athena_corruption_gameplay_cube) (Transformation|GetActorLocation _asbga_athena_corruption_gameplay_cube_2))) -1.0)))
          (bind _output_get_2 (Variables|Default|SetNearestSocketDad _nearestsocket_2))
          (bind (_returnvalue_location_2 _returnvalue_rotation_2) (Math|Transform|BreakTransform (Transformation|GetSocketTransform _dadcubemesh _output_get_2 "RTS_World")))
          (Variables|Default|SetDadSocketPosition _returnvalue_location_2)
          (Effects|Components|ParticleSystem|SetVectorParameter _output "SourcePoint" _returnvalue_location_2)
          (Effects|Components|ParticleSystem|SetVectorParameter _output "SourceNormal" (Math|Rotator|GetRotationXVector _returnvalue_rotation_2))
          (bind _nearestsocket_3 (CallFunction|FindBestSocket _dadcubemesh_2 "0, 0,1.000000"))
          (Effects|Components|ParticleSystem|SetVectorParameter _output "SourceUp" (Math|Rotator|GetRotationXVector (Transformation|GetSocketRotation _dadcubemesh_2 _nearestsocket_3))))
        (:then_2
          (CallFunction|SpawnLoopAudio))))))

(event EventOnRemoval(EitherVFX-preferthisversion)
  (params (MyTarget :type Actor%20Object%20Reference) (Parameters :type Gameplay%20Cue%20Parameters%20Structure%20%28by%20ref%29) (ParticleComponents :type Array%20of%20FXSystem%20Component%20Object%20References) (AudioComponents :type Array%20of%20Audio%20Component%20Object%20References) (BurstCameraShakeInstance :type Camera%20Shake%20Base%20Object%20Reference) (BurstDecalInstance :type Decal%20Actor%20Object%20Reference))
  (CallFunction|SpawnLoopStopAudio)
  (bind _spawneddecal (Variables|Default|GetSpawnedDecal)
    (:then
      (Rendering|Components|Decal|SetFadeOut _spawneddecal 0.0 5.0 true))
    (:else)))

(fn FindBestSocket
  (params (Mesh :type Scene%20Component%20Object%20Reference) (LookAtDirection :type Vector))
  (LocalVariables|SetNearest -1.0)
  (for _array_element (Utilities|Array|MakeArray "A" "B" "C" "D" "E" "F")
    (if (> (Math|Vector|DotProduct (Math|Vector|GetForwardVector (Transformation|GetSocketRotation Mesh _array_element)) LookAtDirection) (LocalVariables|GetNearest))
      (LocalVariables|SetNearest (Math|Vector|DotProduct (Math|Vector|GetForwardVector (Transformation|GetSocketRotation Mesh _array_element)) LookAtDirection))
      (LocalVariables|SetSocket _array_element)))
  (return (LocalVariables|GetSocket)))

(fn SpawnLoopAudio
  (bind _babysocketposition (Variables|Default|GetBabySocketPosition))
  (bind _dadsocketposition (Variables|Default|GetDadSocketPosition))
  (Utilities|FlowControl|Sequence
    (:then_0
      (bind _returnvalue (Audio|SpawnSoundatLocation 0 (Variables|Default|GetBabyAudioLoop) _babysocketposition))
      (Variables|Default|SetBabyAudioComponent _returnvalue)
      (Audio|SpawnSoundatLocation 0 "/CorruptionGameplay/Gameplay/Environmental/Cube/Audio/Cube_Spawn_EndPoint_Cue.Cube_Spawn_EndPoint_Cue" _babysocketposition))
    (:then_1
      (bind _returnvalue_2 (Audio|SpawnSoundatLocation 0 (Variables|Default|GetDadAudioLoop) _dadsocketposition))
      (Variables|Default|SetDadAudioComponent _returnvalue_2)
      (Audio|SpawnSoundatLocation 0 "/CorruptionGameplay/Gameplay/Environmental/Cube/Audio/Cube_Spawn_StartPoint_Cue.Cube_Spawn_StartPoint_Cue" _dadsocketposition))))

(fn SpawnLoopStopAudio
  (Utilities|FlowControl|Sequence
    (:then_0
      (bind _babyaudiocomponent (Variables|Default|GetBabyAudioComponent)
        (:then
          (Audio|Components|Audio|FadeOut _babyaudiocomponent 0.5 0.0 "Linear" :!class /Script/Engine.AudioComponent)
          (Audio|PlaySoundatLocation 0 (Variables|Default|GetBabyAudioLoopStop) (Transformation|GetWorldLocation _babyaudiocomponent)))
        (:else)))
    (:then_1
      (bind _dadaudiocomponent (Variables|Default|GetDadAudioComponent)
        (:then
          (Audio|Components|Audio|FadeOut _dadaudiocomponent 0.5 0.0 "Linear" :!class /Script/Engine.AudioComponent)
          (Audio|PlaySoundatLocation 0 (Variables|Default|GetDadAudioLoopStop) (Transformation|GetWorldLocation _dadaudiocomponent)))
        (:else)))))

(fn ConstructionScript)
