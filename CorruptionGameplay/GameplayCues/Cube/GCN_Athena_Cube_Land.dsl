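; Gameplay cue for the Corruption cube landing, bursting particles, camera shake, force feedback and sound.
; On burst, it moves the first particle component to the cube mesh's ground socket and sets CubeRadius from actor scale.
; Scale also picks one of two shakes and maps the shake and near/distant rumble radii.
; The land sound spawns there with its Size parameter taken from the cue's ability level; everything requires a valid particle component.
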(fn FindBestSocket
  (params (Mesh :type Scene%20Component%20Object%20Reference) (LookAtDirection :type Vector)))

(fn OnBurst(EitherVFX-preferthisversion)
  (params (MyTarget :type Actor%20Object%20Reference) (Parameters :type Gameplay%20Cue%20Parameters%20Structure%20%28by%20ref%29) (ParticleComponents :type Array%20of%20FXSystem%20Component%20Object%20References) (AudioComponents :type Array%20of%20Audio%20Component%20Object%20References) (BurstCameraShakeInstance :type Camera%20Shake%20Base%20Object%20Reference) (BurstDecalInstance :type Decal%20Actor%20Object%20Reference))
  (bind _returnvalue_x (Math|Vector|BreakVector (Transformation|GetActorScale3D MyTarget)))
  (bind (_normalizedmagnitude _rawmagnitude _effectcontext _matchedtagname _originaltag _aggregatedsourcetags _aggregatedtargettags _location _normal _instigator _effectcauser _sourceobject _physicalmaterial _gameplayeffectlevel _abilitylevel) (Ability|GameplayCue|BreakGameplayCueParameters Parameters))
  (bind _returnvalue (Math|Float|MapRangeClamped _returnvalue_x 0.3 1.0 5e+03 7e+03))
  (bind _returnvalue_2 (Math|Float|MapRangeClamped _returnvalue_x 0.3 1.0 1e+03 2e+03))
  (bind _output (Utilities|Array|Get(acopy) ParticleComponents))
  (bind _asbga_athena_corruption_gameplay_cube (Utilities|Casting|CastToBGA_Athena_CorruptionGameplay_Cube MyTarget))
  (Utilities|IsValid _output
    (:Is%20Valid
      (bind (_returnvalue_location _returnvalue_rotation) (Math|Transform|BreakTransform (Transformation|GetSocketTransform (Variables|/CorruptionGameplay/Gameplay/Environmental/Cube/BGA_Athena_CorruptionGameplay_Cube.BGA_Athena_CorruptionGameplay_Cube_C|GetStaticMesh _asbga_athena_corruption_gameplay_cube) (Variables|/CorruptionGameplay/Gameplay/Environmental/Cube/BGA_Athena_CorruptionGameplay_Cube.BGA_Athena_CorruptionGameplay_Cube_C|GetGroundSocket _asbga_athena_corruption_gameplay_cube) "RTS_World")))
      (Transformation|SetWorldLocationAndRotation _output _returnvalue_location _returnvalue_rotation false false)
      (Effects|Components|ParticleSystem|SetFloatParameter _output "CubeRadius" (* _returnvalue_x 768.0))
      (Camera|PlayWorldCameraShake 0 (select (> _returnvalue_x 0.4) "/CorruptionGameplay/Gameplay/Environmental/Cube/CorruptionCubeRollCameraShake.CorruptionCubeRollCameraShake_C" "/FNGameplayCues/Effects/Fort_Effects/BluePrints/CameraShakes/B_Shudder_Vertical_Jolt_CameraShake.B_Shudder_Vertical_Jolt_CameraShake_C") _returnvalue_location _returnvalue_2 _returnvalue)
      (Fortnite|PlayLocalForceFeedbackatLocationMulti 0 "/CorruptionGameplay/Gameplay/Environmental/Cube/FF_Corruption_Cube_Slam.FF_Corruption_Cube_Slam" "/CorruptionGameplay/Gameplay/Environmental/Cube/FF_Corruption_Cube_Slam_Distant.FF_Corruption_Cube_Slam_Distant" _returnvalue_location _returnvalue_2 _returnvalue)
      (bind _returnvalue_3 (Audio|SpawnSoundatLocation 0 (Variables|Default|GetLandSound) _returnvalue_location))
      (Audio|Parameter|SetIntegerParameter _returnvalue_3 "Size" _abilitylevel :!class /Script/AudioExtensions.AudioParameterControllerInterface))))
