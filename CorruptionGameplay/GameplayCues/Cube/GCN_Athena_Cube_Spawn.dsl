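; Gameplay cue that plays the spawn burst VFX for the Corruption cube in Athena.
; On burst, it sets the CubeRadius parameter on a particle component from the target actor's X scale.
; It then casts the target to the cube actor, finds the face socket for the downward direction, and moves that component to the socket's world location and rotation on the cube's static mesh.
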(fn OnBurst(EitherVFX-preferthisversion)
  (params (MyTarget :type Actor%20Object%20Reference) (Parameters :type Gameplay%20Cue%20Parameters%20Structure%20%28by%20ref%29) (ParticleComponents :type Array%20of%20FXSystem%20Component%20Object%20References) (AudioComponents :type Array%20of%20Audio%20Component%20Object%20References) (BurstCameraShakeInstance :type Camera%20Shake%20Base%20Object%20Reference) (BurstDecalInstance :type Decal%20Actor%20Object%20Reference))
  (bind _returnvalue_x (Math|Vector|BreakVector (Transformation|GetActorScale3D MyTarget)))
  (bind _output (Utilities|Array|Get(acopy) ParticleComponents))
  (Effects|Components|ParticleSystem|SetFloatParameter _output "CubeRadius" (* _returnvalue_x 768.0))
  (bind _asbga_athena_corruption_gameplay_cube (Utilities|Casting|CastToBGA_Athena_CorruptionGameplay_Cube MyTarget))
  (bind _facesocket (Class|BGAAthenaCorruptionGameplayCube|FindFaceSocket _asbga_athena_corruption_gameplay_cube "0.000000,0.000000,-1.000000"))
  (bind (_returnvalue_location _returnvalue_rotation) (Math|Transform|BreakTransform (Transformation|GetSocketTransform (Variables|/CorruptionGameplay/Gameplay/Environmental/Cube/BGA_Athena_CorruptionGameplay_Cube.BGA_Athena_CorruptionGameplay_Cube_C|GetStaticMesh _asbga_athena_corruption_gameplay_cube) _facesocket "RTS_World")))
  (Transformation|SetWorldLocationAndRotation _output _returnvalue_location _returnvalue_rotation false false))
